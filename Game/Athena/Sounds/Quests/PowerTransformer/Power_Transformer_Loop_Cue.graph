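# Sound Cue for the looping hum of a power transformer, routed to the quest sound class.
# A Random node picks between two looping wave players with equal weight, so each playback settles on one of two loop variations.
# It references a dedicated transformer attenuation asset for spatial falloff and is primed on load.
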SoundCue /Game/Athena/Sounds/Quests/PowerTransformer/Power_Transformer_Loop_Cue
root=Random_0
nodes [
 {id=Random_0, children [WavePlayer_2,WavePlayer_3], properties {weights [1,1]}}
 {id=WavePlayer_2, properties {bLooping=true, soundWaveAssetPtr=&/Power_Transformer_Loop_02}}
 {id=WavePlayer_3, properties {bLooping=true, soundWaveAssetPtr=&/Power_Transformer_Loop_01}}
]
properties {
 attenuationSettings=&/Fort_Power_Transformer_Loop_Att
 bPrimeOnLoad=true
 maxDistance=2550
 soundClassObject=&Fort_Quests
 volumeMultiplier=1
 attenuationOverrides {falloffDistance=700, attenuationShapeExtents {x=100}}
}
