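# Sound cue for a successful interaction with the Power Transformer in an Athena quest.
# A Random node picks one of four equally weighted wave variations, and a Modulator adds slight volume variation.
# It uses the non-critical in-game UI sound class, with a concurrency override that allows one instance and stops the oldest on retrigger.
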SoundCue /Game/Athena/Sounds/Quests/PowerTransformer/Power_Transformer_Succesful_Interaction_Cue
root=Modulator_0
nodes [
 {id=Modulator_0, children [Random_1], properties {volumeMax=1.1, volumeMin=1}}
 {id=Random_1, children [WavePlayer_1,WavePlayer_5,WavePlayer_6,WavePlayer_7], properties {weights [1,1,1,1]}}
 {id=WavePlayer_1, properties {soundWaveAssetPtr=&/Power_Transformer_Successful_Interaction_04}}
 {id=WavePlayer_5, properties {soundWaveAssetPtr=&/Power_Transformer_Successful_Interaction_01}}
 {id=WavePlayer_6, properties {soundWaveAssetPtr=&/Power_Transformer_Successful_Interaction_02}}
 {id=WavePlayer_7, properties {soundWaveAssetPtr=&/Power_Transformer_Successful_Interaction_03}}
]
properties {
 bOverrideConcurrency=true
 maxDistance=2097152
 soundClassObject=&Fort_UI_InGame_NonCritical
 volumeMultiplier=1
 concurrencyOverrides {maxCount=1, resolutionRule=StopOldest}
}
